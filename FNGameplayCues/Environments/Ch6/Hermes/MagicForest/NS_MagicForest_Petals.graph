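# Niagara ambient effect for the Magic Forest that blows pink-purple petals across a masked open-field area and adds circling butterflies.
# GPU petals spawn at a fixed map position via importance sampling of a field mask and landscape sampling, with NPC wind-gust state driving rate and wind.
# Butterflies spawn in two height-offset partitions and swirl under vortex and line-attraction forces.
# The DEBUG and owner-movement petal trail emitters are disabled.
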NiagaraSystem /FNGameplayCues/Environments/Ch6/Hermes/MagicForest/NS_MagicForest_Petals
dependencies {
 usedDataInterfaces [
  &NiagaraDataInterfaceColorCurve
  /Script/Niagara.NiagaraDataInterfaceTexture
  &NiagaraDataInterfaceLandscape
  &NiagaraDataInterfaceCurve
  /Script/Niagara.NiagaraDataInterfaceEmitterProperties
  &NiagaraDataInterfaceCollisionQuery
  &NiagaraDataInterfaceGBuffer
  &NiagaraDataInterfaceStaticMesh
  &NiagaraDataInterfaceArrayFloat3
 ]
 usedDynamicInputs [&MaskVectorBySpawnGroup]
 usedModules [&SystemState, &Timeline]
}
emitters {
 Butterfly {
  emitterUpdateScript [
   &EmitterState(Enable Distance Culling=true, MaxDistance=8000, Module.bUseMaxDistance=true, Module.bUseMinDistance=false, Scalability Mode=::NewEnumerator1)
   &SpawnRate(SpawnRate=4, Use Spawn Probability=false)
  ]
  particleSpawnScript [
   &PartitionParticles()
   &InitializeParticle(
    Color Mode=::NewEnumerator0
    Lifetime=&RandomRangeFloat(Minimum=3, Maximum=4)
    Mesh Scale Mode=::NewEnumerator3
    Mesh Uniform Scale=0.03
    UsePositionOffset=false
    Write Material Random=true
   )
   &InitialMeshOrientation(Mesh Axis To Orient=<0 0 1>, Mesh Orientation Mode=::NewEnumerator2, Use Rotation Vector=false)
   &ShapeLocation(
    Cylinder Height=50
    Cylinder Radius=&Multiply_Float(A=800, B=$User.ShapeScale)
    MaskSpawnGroup=false
    Offset=&SelectVectorFromArray(
     Direct Array Index=$Particles.PartitionParticles.Partition
     Vector Selection Array=DataInterface() {propertyValues {FloatData [<0 0 0>, <0 0 600>]}}
    )
    Offset Mode=::NewEnumerator0
    Override Seed=false
    Shape Primitive=::NewEnumerator1
    Surface Only=true
    Use Endcaps In Surface Only Mode=false
    UseOwnerScale=true
   ) {enabled=false}
   &ShapeLocation(
    Handle Radius=200
    Large Radius=1200
    MaskSpawnGroup=false
    Offset=&SelectVectorFromArray(
     Direct Array Index=$Particles.PartitionParticles.Partition
     Vector Selection Array=DataInterface() {propertyValues {FloatData [<0 0 0>, <0 0 800>]}}
    )
    Offset Mode=::NewEnumerator0
    Override Seed=false
    Shape Primitive=::NewEnumerator3
    UseOwnerScale=true
   ) {module=ShapeLocation001}
   &AddVelocity(Velocity=&SelectVectorFromArray(
     Direct Array Index=$Particles.PartitionParticles.Partition
     Vector Selection Array=DataInterface() {propertyValues {FloatData [<0 0 50>, <0 0 75>]}}
    ))
   &ApplyInitialForces() {enabled=false}
  ]
  particleUpdateScript [
   &ParticleState()
   &VortexForce(Influence Falloff Radius=1750, Origin Pull Amount=30, Use Influence Falloff=true, UseOriginOffset=false, Vortex Force Amount=-50)
   /Niagara/Modules/Update/Forces/LineAttractionForce(
    Attraction Strength=50
    Line End:Vector3f=&Add_Vector(
     A=Linked() {linkedVariable {name=Engine.Owner.Position, type=NiagaraPosition}}
     B=&SelectVectorFromArray(
      Direct Array Index=$Particles.PartitionParticles.Partition
      Vector Selection Array=DataInterface() {propertyValues {FloatData [<0 0 -600>, <0 0 2000>]}}
     )
    )
    Line Start:NiagaraPosition=$Engine.Owner.Position
    Use Attraction Falloff=false
   )
   &DynamicMaterialParameters(
    Index 0 Param 1=0.2364
    Index 0 Param 4=&Multiply_Float(A=$System.NormalizedLoopAge, B=3)
    Index 1 Param 1=3000
    Index 1 Param 2=1
    Module.Param0WriteEnabled=true
    Module.Param0WriteEnabled1=true
    Module.Param1WriteEnabled=true
    Module.Param1WriteEnabled1=true
    Module.Param2WriteEnabled=true
    Module.Param2WriteEnabled1=true
    Module.Param3WriteEnabled=true
    Module.Param3WriteEnabled1=true
    Write Parameter Index 1=true
   )
   &SolveForcesAndVelocity(Clamp Velocity=false, Limit Acceleration=false)
   &ScaleMeshSize(Scale Factor=&VectorFromFloat(Value=&FloatFromCurve(FloatCurve=DataInterface() {propertyValues {Curve {keys [
          <at 0 0 RCIM_Cubic>
          {arriveTangent=1.1765, interpMode=RCIM_Cubic, leaveTangent=1.1765, time=0.15, value=1}
          {arriveTangent=-1.1765, interpMode=RCIM_Cubic, leaveTangent=-1.1765, time=0.85, value=1}
          <at 1 0 RCIM_Cubic>
         ]}}})))
   &FN_Sprite_Attributes(
    Color=<rgba 0 0.5183 1 0.1>
    Create Color Attribute=true
    Create SpriteSize Attribute=true
    SpriteSize=&Multiply_Vector2D_ByFloat(Float=&MakeFloatFromVector(Vector=$Particles.Scale, Channel=::NewEnumerator0), Vector2D=<1200 1200>)
    Write to Module Namespace=true
   )
  ]
  rendererClasses [&NiagaraSpriteRendererProperties, &NiagaraMeshRendererProperties]
  renderers [
   {rendererClass=&NiagaraSpriteRendererProperties}
   {rendererClass=&NiagaraMeshRendererProperties, rendererIndex=1}
  ]
 }
 DEBUG {
  bEnabled=false
  simTarget=GPUComputeSim
  emitterUpdateScript [
   &EmitterState()
   &SpawnRate(SpawnRate=150000, Use Spawn Probability=false)
  ]
  particleSpawnScript [
   &AddVelocity(Velocity=<0 50 0>)
   &InitializeParticle(
    Color Maximum=<rgba 0 0.1413 0.0949 1>
    Color Minimum=<rgba 0 1 0.2085 1>
    Color Mode=::NewEnumerator2
    Lifetime Max=1
    Lifetime Mode=::NewEnumerator1
    Mesh Scale Mode=::NewEnumerator1
    Mesh Uniform Scale Max=10
    Mesh Uniform Scale Min=6
    Sprite Size Mode=::NewEnumerator1
    Sprite UV Mode=::NewEnumerator1
    Uniform Sprite Size Min=8
    UsePositionOffset=false
    Write Material Random=true
   )
   &ShapeLocation(
    Cylinder Radius=150000
    MaskSpawnGroup=false
    Override Seed=false
    Shape Primitive=::NewEnumerator1
    Surface Only=true
    Surface Only Band Thickness=1
    Use Endcaps In Surface Only Mode=false
    UseOwnerScale=true
   )
   /Niagara/Modules/Math/MoveToNearestDistanceFieldSurface_GPU(Kill by Distance=5000, Offset Particle By Radius=true, Radius=1500, bKillByDistance=true) {enabled=false}
   /SetVariables(Particles.InitialPos:NiagaraPosition=$Particles.Position) {isSetParameters=true}
   /Niagara/Modules/Math/PlaceParticlesOnDepthBuffer_GPU(Max Z Depth=5000, Module.Use Max Z Depth=true) {enabled=false}
   &self:ScratchModule_02(Angle=25, Normal:Vector3f=<0 0 1>) {enabled=false, module=ScratchModule_02}
   /SetVariables(Particles.Position=&AddVectorToPosition(Position=$Particles.Position, Vector=<0 0 25>)) {enabled=false, isSetParameters=true}
   /Game/Effects/Niagara/Modules/Spawn/FN_Spawn_GBuffer_Color_WithinRangeV2(HueMax=185, HueMin=175) {enabled=false}
   &KillParticles(Kill Particles=&SetBoolByFloatComparison(
     A=&VectorLength(Vector=&Subtract_Vector(A=$Particles.InitialPos, B=Linked() {linkedVariable {name=Particles.Position, type=NiagaraPosition}}))
     B=5000
    )) {enabled=false}
   &DynamicMaterialParameters(
    Index 0 Param 1=&RandomRangeFloat(Minimum=100, Maximum=50)
    Index 0 Param 2=5
    Index 0 Param 3=15
    Index 0 Param 4=&RandomRangeFloat(Minimum=0, Maximum=0)
    Index 1 Param 1=25
    Index 2 Param 1=&RandomRangeFloat(Minimum=-0.1, Maximum=-0.1)
    Index 2 Param 2=&RandomRangeFloat(Minimum=0.1, Maximum=0)
    Index 2 Param 3=&RandomRangeFloat(Minimum=0, Maximum=0)
    Index 2 Param 4=&RandomRangeFloat(Minimum=0.2, Maximum=0.3)
    Module.Index0Param0Write=true
    Module.Index0Param1Write=true
    Module.Index0Param2Write=true
    Module.Index0Param3Write=true
    Module.Index1Param0Write=true
    Module.Index1Param1Write=true
    Module.Index1Param2Write=true
    Module.Index1Param3Write=true
    Module.Index2Param0Write=true
    Module.Index2Param1Write=true
    Module.Index2Param2Write=true
    Module.Index2Param3Write=true
    Write Parameter Index 1=true
    Write Parameter Index 2=true
   )
   /SetVariables(Particles.Linear Color=<rgba 1 0 0 1>, Particles.WaveScale:Vector3f=<6 6 6>) {isSetParameters=true}
   &CameraOffset(Camera Offset Amount=5)
   &KillParticlesInVolume(
    Box Size:Vector3f=<15000 15000 250000>
    Kill Shape=::NewEnumerator1
    Volume Origin:Vector3f=&AddVectorToPosition(Position=<49038 -89768 3140>, Vector=<0 0 0>)
   ) {enabled=false}
   &self:RainTextureMasking_02(
    AlphaThreshold=0.1
    BoxCenter:NiagaraPosition=<-66244 10607.583 13622.6396>
    BoxExtents:Vector3f=<130000 130000 1>
    Position:NiagaraPosition=$Particles.Position
    RejectionTexture=DataInterface() {propertyValues {
      Texture=/Game/Environments/Ch6/Hermes/MagicForest/Materials/T_OpenField_FXMask
      TextureUserParameter {parameter {name=None, typeDefHandle {registeredTypeIndex=2}, varData []}}
     }}
   ) {module=RainTextureMasking_02}
   /Niagara/Modules/Spawn/Location/StaticMeshLocation(
    MaskSpawnGroup=false
    Mesh Sampling Type:ENiagara_StaticSamplingMode=::NewEnumerator2
    OffsetAlongNormal=false
    SamplePosOffset=false
    Static Mesh=DataInterface() {propertyValues {
      DefaultMesh=null
      InstanceIndex=-1
      LODIndex=0
      PreviewMesh=null
      SoftSourceActor=null
      SourceMode=Default
      bAllowSamplingFromStreamingLODs=false
      bCaptureTransformsPerFrame=true
      bUsePhysicsBodyVelocity=false
      FilteredSockets []
      LODIndexUserParameter {parameter {name=None, typeDefHandle {registeredTypeIndex=108}, varData []}}
      MeshParameterBinding {parameter {name=None, typeDefHandle {registeredTypeIndex=1}, varData []}}
      SectionFilter {allowedMaterialSlots []}
     }}
    Vertex ID:NiagaraInt32=$Engine.ExecIndex
    Vertex Sampling Mode:ENiagara_StaticVertexMode=::NewEnumerator1
   )
  ]
  particleUpdateScript [
   &ParticleState()
   &CurlNoiseForce(
    Curl Noise Cone Mask Angle=25
    Mask Curl Noise=true
    Noise Frequency=&RandomRangeFloat(Minimum=2, Maximum=4)
    Noise Strength=&Multiply_Float(A=&RandomRangeFloat(Minimum=25, Maximum=100), B=1)
    Pan Noise=true
   )
   &WindForce(Use Depth Buffer=false, Use Landscape=false, Wind Speed=<0 25 0>)
   &Drag(Drag=0.1, Use Linear Drag=true, Use Rotational Drag=true)
   &SolveForcesAndVelocity(
    Acceleration Limit=25
    Clamp Velocity=true
    Force:Vector3f=&Multiply_Vector(A=$Transient.PhysicsForce, B=<1 1 0>)
    Limit Acceleration=true
    Speed Limit=25
    Velocity:Vector3f=&Multiply_Vector(A=$Particles.Velocity, B=<1 1 0>)
   )
   &CalculateAccurateVelocity() {enabled=false}
   &CurlNoiseForce(
    Curl Noise Cone Mask Angle=80
    Curl Noise Cone Mask Falloff Angle=90
    Mask Curl Noise=true
    Noise Frequency=&RandomRangeFloat(Minimum=0.25, Maximum=1)
    Noise Strength=&Multiply_Float(A=&RandomRangeFloat(Minimum=35, Maximum=100), B=1)
    Pan Noise=false
    Pan Noise Field=<0 0 2>
   ) {module=CurlNoiseForce001}
   /Niagara/Modules/Math/AvoidDistanceFieldSurfaces_GPU(
    Avoid Nearest Surface=true
    Avoid Oncoming Surfaces=true
    Falloff Distance=1
    Max Trace Distance=1
    Nearest Surface Avoidance Strength=1
    Obstactle Avoidance Strength=1
    Trace Vector:Vector3f=&Multiply_Vector(A=$Particles.Velocity, B=<1 1 0>)
   )
   &SolveForcesAndVelocity(
    Acceleration Limit=5
    Clamp Velocity=true
    Force:Vector3f=&Multiply_Vector(A=$Transient.PhysicsForce, B=<1 1 0>)
    Limit Acceleration=true
    Speed Limit=100
    Velocity:Vector3f=&Multiply_Vector(A=$Particles.Velocity, B=<1 1 0>)
   ) {module=SolveForcesAndVelocity001}
   &RegionMask(
    Invert Region=true
    Offset:Vector3f=<49038 -89768 3140>
    Offset Space:ENiagaraCoordinateSpace=::Local
    Region Dimensions:Vector3f=<15000 15000 15000>
    Region Type:ENiagaraDebugDrawShape=::NewEnumerator1
    UseOffset=true
    UseRotation=false
   )
   &ScaleColor(
    Scale Alpha=&FloatFromCurve(
     FloatCurve=DataInterface() {propertyValues {Curve {keys [
         {interpMode=RCIM_Cubic, tangentMode=RCTM_User, time=0, value=0}
         {interpMode=RCIM_Cubic, tangentMode=RCTM_User, time=0.5, value=1}
         {tangentMode=RCTM_User, time=1, value=0}
        ]}}}
     Scale Curve=&Multiply_Float(A=0.25, B=$StackContext.RegionMask.Float)
    )
    Scale RGB=&Multiply_VectorByFloat(Vector=<1 1 1>, Float=&RandomRangeFloat(Minimum=1, Maximum=4))
    ScaleA=true
    ScaleRGB=true
   ) {module=ScaleColor001}
   /SetVariables(
    Particles.Linear Color=&MultiplyLinearColor(
     Linear Color=$Particles.Color
     Scale Factor=&MakeVector4_FromFloat(Float=&FloatFromCurve(
       FloatCurve=DataInterface() {propertyValues {Curve {keys [
           {interpMode=RCIM_Cubic, tangentMode=RCTM_User, time=0, value=0}
           {interpMode=RCIM_Cubic, tangentMode=RCTM_User, time=0.5, value=1}
           {tangentMode=RCTM_User, time=1, value=0}
          ]}}}
       Scale Curve=$System.ColorIntensityMult
      ))
    )
    Particles.Linear Color001:LinearColor=&MultiplyLinearColor(
     Linear Color=&ColorFromCurve(ColorCurve=DataInterface() {propertyValues {
        AlphaCurve {keys [<at 0 0 RCIM_Cubic>,<at 0.1218 1>,<at 1 0 RCIM_Cubic>]}
        BlueCurve {keys [<at 0 0.0688 RCIM_Cubic>,<at 1 1 RCIM_Cubic>]}
        GreenCurve {keys [<at 0 1 RCIM_Cubic>,<at 1 1 RCIM_Cubic>]}
        RedCurve {keys [<at 0 0 RCIM_Cubic>,<at 1 1 RCIM_Cubic>]}
       }})
     Scale Factor=&MakeVector4_FromFloat(Float=&FloatFromCurve(
       FloatCurve=DataInterface() {propertyValues {Curve {keys [<at 0 1>,<at 1 0>]}}}
       Scale Curve=&Multiply_Float(A=$System.ColorIntensityMult, B=4)
      ))
    )
   ) {isSetParameters=true}
   &CameraOffset() {module=CameraOffset001}
  ]
  rendererClasses [&NiagaraSpriteRendererProperties]
  renderers [{rendererClass=&NiagaraSpriteRendererProperties}]
 }
 Petals {
  simTarget=GPUComputeSim
  emitterSpawnScript [/SetVariables(Emitter.Camera Position:NiagaraPosition=<-65871 11037 3443>) {isSetParameters=true, module=SetVariables_#1}]
  emitterUpdateScript [
   &EmitterState()
   &SpawnRate(
    SpawnRate=&Multiply_Float(A=750, B=&Lerp_Float(A=0.25, B=2, Alpha=$NPC.FortniteNPC.WindGustsSystem_Active))
    Use Spawn Probability=false
   )
  ]
  particleSpawnScript [
   &InitializeParticle(
    Color=&ColorFromCurve(
     ColorCurve=DataInterface() {propertyValues {
       AlphaCurve {keys [<at 0 1 RCIM_Cubic>]}
       BlueCurve {keys [<at 0 0.5776 RCIM_Cubic>,<at 0.4861 0.5583>,<at 1 0.2086 RCIM_Cubic>]}
       GreenCurve {keys [<at 0 0.1329 RCIM_Cubic>,<at 0.4861 0.1301>,<at 1 0.0513 RCIM_Cubic>]}
       RedCurve {keys [<at 0 0.3095 RCIM_Cubic>,<at 0.4861 0.3325>,<at 1 0.1384 RCIM_Cubic>]}
      }}
     CurveIndex=&RandomRangeFloat(Minimum=0, Maximum=1)
    )
    Lifetime=5
    Mesh Scale Mode=::NewEnumerator3
    Mesh Uniform Scale=&RandomRangeFloat(Minimum=0.3, Maximum=0.45)
    Position=&Subtract_Position(A=$Emitter.Camera Position, B=$Engine.Owner.Position)
    Position Mode=::NewEnumerator1
    Sprite Size Mode=::NewEnumerator3
    Uniform Sprite Size=50
    UsePositionOffset=false
    Write Material Random=true
   )
   &ShapeLocation(
    MaskSpawnGroup=false
    Non Uniform Scale=<1 1 0>
    Override Seed=false
    Sphere Radius=&MaskFloatBySpawnGroup(Masked Float=2500, Passthrough Float=100)
    UseOwnerScale=true
   )
   &InitialMeshOrientation(Use Rotation Vector=false)
   /Niagara/Modules/Update/Textures/SampleTexture(
    "Comparison Value (B)"=0.5
    Enable Sample Rejection=true
    EnableFalloff=true
    Generate UV Using Importance Sampling=true
    Rotation Coordinate Space:ENiagaraCoordinateSpace=::World
    Rotation Mode:ENiagara_RotationMode=::NewEnumerator0
    Texture=DataInterface() {propertyValues {
      Texture=&T_OpenField_FXMask
      TextureUserParameter {parameter {name=None, typeDefHandle {registeredTypeIndex=2}, varData []}}
     }}
    Transform UV Into Position=true
    UV Origin:NiagaraPosition=&Add_Vector(A=Linked() {linkedVariable {name=Emitter.Camera Position, type=NiagaraPosition}}, B=<0 -600 0>)
    UV Scale:Vector2f=<13500 13500>
   )
   &self:SampleLandscape() {module=SampleLandscape}
   &KillParticles(Kill Particles=&SetBoolByFloatComparison(
     A=&MakeFloatFromVector4(
      Vector 4=&MakeVector4_FromLinearColor(LinearColor=$Particles.SampleTexture.SampledColor)
      Channel=::NewEnumerator1
     )
     B=0.5
     Comparison Type=::NewEnumerator6
    )) {enabled=false}
   &Color(Color=&ColorFromCurve(
     ColorCurve=DataInterface() {propertyValues {
       AlphaCurve {keys [<at 0 1 RCIM_Cubic>]}
       BlueCurve {keys [<at 0 0.5776 RCIM_Cubic>,<at 0.4861 0.5583>,<at 1 0.2086 RCIM_Cubic>]}
       GreenCurve {keys [<at 0 0.1329 RCIM_Cubic>,<at 0.4861 0.1301>,<at 1 0.0513 RCIM_Cubic>]}
       RedCurve {keys [<at 0 0.3095 RCIM_Cubic>,<at 0.4861 0.3325>,<at 1 0.1384 RCIM_Cubic>]}
      }}
     CurveIndex=&RandomRangeFloat(Minimum=0, Maximum=1)
    ))
   &CameraOffset(Camera Offset Amount=-100)
  ]
  particleUpdateScript [
   &ParticleState()
   &ScaleMeshSize(Scale Factor=&VectorFromFloat(Value=&FloatFromCurve(FloatCurve=DataInterface() {propertyValues {Curve {keys [{interpMode=RCIM_Cubic, tangentMode=RCTM_Break, time=0, value=1}, <at 1 0>]}}})))
   &UpdateMeshOrientation()
   &ScaleSpriteSize(Uniform Curve Sprite Scale=DataInterface() {propertyValues {Curve {keys [{interpMode=RCIM_Cubic, tangentMode=RCTM_Break, time=0, value=1}, {tangentMode=RCTM_User, time=1, value=0}]}}})
   &self:EmitterBounds(
    Emitter Properties=DataInterface() {propertyValues {EmitterBinding {bindingMode=Self, emitterName=None}}}
    EmitterLocation:NiagaraPosition=$Particles.Position
   ) {module=EmitterBounds}
   &WindForce(
    Friction Mode:ENiagaraWindFrictionMode=::NewEnumerator0
    Scale=1
    Surface Alignment Mode:ENiagaraWindCombingMode=::NewEnumerator0
    Wind Speed=<300 0 50>
    Wind Speed Scale=&Lerp_Float(A=0.75, B=1, Alpha=$NPC.FortniteNPC.WindGustsSystem_Active)
   )
   &Drag(Drag=5, Use Linear Drag=true, Use Rotational Drag=true)
   &GravityForce(Gravity=<0 0 -360>)
   &SolveForcesAndVelocity(Clamp Velocity=false, Limit Acceleration=false)
   &ScaleColor(ScaleA=true, ScaleRGB=true)
   /SetVariables(Particles.Orientation:Quat4f=&MultiplyQuaternion(
     Quaternion A=&FindQuatBetweenTwoVectors(
      From Vector=<1 0 0>
      To Vector=&NormalizeVector(Vector To Normalize=$Particles.Velocity, Safe Normalize=true, Safe Fallback Threshold=0, Safe Fallback Vector=<0 0 0.001>)
     )
     Quaternion B=<0 0 0 1>
    )) {isSetParameters=true, module=SetVariables_#2}
   &DynamicMaterialParameters(
    Index 0 Param 1=0.5
    Module.Index0Param0Write=true
    Module.Index0Param1Write=true
    Module.Index0Param2Write=true
    Module.Index0Param3Write=true
   )
  ]
  rendererClasses [&NiagaraMeshRendererProperties, &NiagaraSpriteRendererProperties]
  renderers [
   {rendererClass=&NiagaraMeshRendererProperties, rendererIndex=0}
   {rendererClass=&NiagaraSpriteRendererProperties, rendererIndex=1}
  ]
 }
 Petals001 {
  bEnabled=false
  simTarget=GPUComputeSim
  emitterSpawnScript [/SetVariables(Emitter.Camera Position:NiagaraPosition=<-65871 11037 3443>) {isSetParameters=true, module=SetVariables_#1}]
  emitterUpdateScript [
   &EmitterState()
   &SpawnRate(SpawnRate=50, Use Spawn Probability=false) {enabled=false}
   &SpawnPerUnit(
    Module.Use Max Movement Threshold=false
    Module.Use Movement Tolerance=false
    Spawn Group=1
    Use Spawn Probability=false
    Velocity Vector:Vector3f=&Multiply_Vector(A=$Engine.Owner.Velocity, B=<1 1 0.1>)
   )
  ]
  particleSpawnScript [
   &InitializeParticle(
    Color=<rgba 0.5878 0.1521 1 1>
    Lifetime=5
    Mesh Scale Mode=::NewEnumerator1
    Mesh Uniform Scale Max=0.25
    Mesh Uniform Scale Min=0.1
    Sprite Size Mode=::NewEnumerator3
    Uniform Sprite Size=50
    UsePositionOffset=false
    Write Material Random=true
   )
   &ShapeLocation(
    MaskSpawnGroup=false
    Non Uniform Scale=<1 1 0>
    Offset=<0 0 30>
    Offset Mode=::NewEnumerator0
    Override Seed=false
    Sphere Radius=10
    UseOwnerScale=true
   )
   &InitialMeshOrientation(Use Rotation Vector=false)
   &self:RainTextureMasking_01(
    AlphaThreshold=0.5
    BoxCenter:NiagaraPosition=<-65871 11037 38212.6797>
    BoxExtents:Vector3f=<135000 135000 1>
    Position:NiagaraPosition=$Particles.Position
    RejectionTexture=DataInterface() {propertyValues {
      Texture=&T_OpenField_FXMask
      TextureUserParameter {parameter {name=None, typeDefHandle {registeredTypeIndex=2}, varData []}}
     }}
   ) {enabled=false, module=RainTextureMasking_01}
   /SetVariables() {enabled=false, isSetParameters=true}
   &Color(Color=&ColorFromCurve(
     ColorCurve=DataInterface() {propertyValues {
       AlphaCurve {keys [<at 0 1 RCIM_Cubic>]}
       BlueCurve {keys [<at 0 0.2462 RCIM_Cubic>,<at 1 0.1683 RCIM_Cubic>]}
       GreenCurve {keys [<at 0 0.0931 RCIM_Cubic>,<at 1 0.0513 RCIM_Cubic>]}
       RedCurve {keys [<at 0 0.1651 RCIM_Cubic>,<at 1 0.0953 RCIM_Cubic>]}
      }}
     CurveIndex=&RandomRangeFloat(Minimum=0, Maximum=1)
    ))
   &AddVelocity(Module.Use Falloff Distance=false, Override Seed=false, Velocity Mode=::NewEnumerator1, Velocity Speed=250)
  ]
  particleUpdateScript [
   &ParticleState()
   &ScaleMeshSize(Scale Factor=&VectorFromFloat(Value=&FloatFromCurve(FloatCurve=DataInterface() {propertyValues {Curve {keys [{interpMode=RCIM_Cubic, tangentMode=RCTM_Break, time=0, value=1}, <at 1 0>]}}})))
   &UpdateMeshOrientation()
   &ScaleSpriteSize(Uniform Curve Sprite Scale=DataInterface() {propertyValues {Curve {keys [{interpMode=RCIM_Cubic, tangentMode=RCTM_Break, time=0, value=1}, {tangentMode=RCTM_User, time=1, value=0}]}}})
   &self:EmitterBounds(
    Emitter Properties=DataInterface() {propertyValues {EmitterBinding {bindingMode=Self, emitterName=None}}}
    EmitterLocation:NiagaraPosition=$Particles.Position
   ) {module=EmitterBounds}
   &WindForce(
    Friction Mode:ENiagaraWindFrictionMode=::NewEnumerator0
    Scale=1
    Surface Alignment Mode:ENiagaraWindCombingMode=::NewEnumerator0
    Wind Speed=<300 0 0>
   )
   &Drag(Drag=5, Use Linear Drag=true, Use Rotational Drag=true)
   &GravityForce(Gravity=<0 0 -360>)
   &SolveForcesAndVelocity(Clamp Velocity=false, Limit Acceleration=false)
   &ScaleColor(ScaleA=true, ScaleRGB=true)
   /SetVariables(Particles.Orientation:Quat4f=&MultiplyQuaternion(
     Quaternion A=&FindQuatBetweenTwoVectors(
      From Vector=<1 0 0>
      To Vector=&NormalizeVector(Vector To Normalize=$Particles.Velocity, Safe Normalize=true, Safe Fallback Threshold=0, Safe Fallback Vector=<0 0 0.001>)
     )
     Quaternion B=<0 0 0 1>
    )) {isSetParameters=true, module=SetVariables_#2}
  ]
  rendererClasses [&NiagaraMeshRendererProperties, &NiagaraSpriteRendererProperties]
  renderers [
   {rendererClass=&NiagaraMeshRendererProperties, rendererIndex=0}
   {rendererClass=&NiagaraSpriteRendererProperties, rendererIndex=1}
  ]
 }
}
summary {
 systemName=NS_MagicForest_Petals
 emitters [
  {
   emitterName=Petals
   simTarget=GPUComputeSim
   rendererClasses [&NiagaraMeshRendererProperties, &NiagaraSpriteRendererProperties]
  }
  {bEnabled=false, emitterName=DEBUG, simTarget=GPUComputeSim, rendererClasses [&NiagaraSpriteRendererProperties]}
  {
   bEnabled=false
   emitterName=Petals001
   simTarget=GPUComputeSim
   rendererClasses [&NiagaraMeshRendererProperties, &NiagaraSpriteRendererProperties]
  }
  {emitterName=Butterfly, rendererClasses [&NiagaraSpriteRendererProperties, &NiagaraMeshRendererProperties]}
 ]
 userVariables [
  User.CaptureMatrix {
   type=NiagaraMatrix
   description=""
   defaultValue {struct=NiagaraMatrix, value {row0=<1 0 0 0>, row1=<0 1 0 0>, row2=<0 0 1 0>, row3=<0 0 0 1>}}
  }
  User.CaptureOrthoWidth {type=NiagaraFloat, defaultValue=512, description=""}
  User.ShapeScale {type=NiagaraFloat, defaultValue=1, description=""}
 ]
}
